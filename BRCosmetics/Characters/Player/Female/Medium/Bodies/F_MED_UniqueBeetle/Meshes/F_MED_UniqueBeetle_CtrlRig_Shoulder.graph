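# Control Rig for the F_MED_UniqueBeetle character body that adds corrective motion to shoulder pieces, kneepads and boots.
# It runs FortDeformRig, then sequences clavicle, right arm, left arm and leg correctives.
# Spherical pose readers and a DeltaRotation twist function, measured against setup-created clavicle-child nulls, weight additive bone offsets.
# Shoulders_02 bones also use parent and rotation constraints; some older shoulder chains remain disconnected.
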ControlRigBlueprint /BRCosmetics/Characters/Player/Female/Medium/Bodies/F_MED_UniqueBeetle/Meshes/F_MED_UniqueBeetle_CtrlRig_Shoulder
graphs [
 New Function_ContainedGraph {nodes [
   n1 = RigVMFunctionEntryNode()
   RigVMFunction_MathBoolToInteger = &RigVMFunction_MathBoolToInteger(Value=@n1.onlyTwist)
   RigVMFunction_MathQuaternionSwingTwist = &RigVMFunction_MathQuaternionSwingTwist(
    twistAxis=@n1.twistAxis
    Input=&RigUnit_GetRelativeTransformForItem(bChildInitial=False, bParentInitial=False, child=@n1.child, parent=@n1.parent).RelativeTransform.Rotation
   )
   n2 = RigVMDispatchNode(
    Values=((),())
    Index=@RigVMFunction_MathBoolToInteger.Result
    Values.0=@RigVMFunction_MathQuaternionSwingTwist.Swing
    Values.1=@RigVMFunction_MathQuaternionSwingTwist.Twist
   ) {template="DISPATCH_RigVMDispatch_SelectInt32(in Index,in Values,out Result)"}
   RigVMFunction_MathQuaternionToEuler = &RigVMFunction_MathQuaternionToEuler(Value=@n2.Result, RotationOrder=@n1.Rotation Order)
   > DISPATCH_RigVMDispatch_Print = RigVMDispatchNode(ScreenColor=(R=1,G=1,B=1,A=1), ScreenDuration=1.5, Enabled=@n1.debugRotation, Value=@RigVMFunction_MathQuaternionToEuler.Result) {template="DISPATCH_RigVMDispatch_Print(in Prefix,in Value,in Enabled,in ScreenDuration,in ScreenColor)"}
   Cast = RigVMDispatchNode(Value=@n1.triggerAxis) {template="DISPATCH_RigVMDispatch_CastEnumToInt(in Value,out Result)"}
   DISPATCH_RigVMDispatch_SelectInt32 = RigVMDispatchNode(
    Values=(0,0,0)
    Index=@Cast.Result
    Values.0=@RigVMFunction_MathQuaternionToEuler.Result.X
    Values.1=@RigVMFunction_MathQuaternionToEuler.Result.Y
    Values.2=@RigVMFunction_MathQuaternionToEuler.Result.Z
   ) {template="DISPATCH_RigVMDispatch_SelectInt32(in Index,in Values,out Result)"}
   Remap = &RigVMFunction_MathDoubleRemap(
    bClamp=False
    Value=@DISPATCH_RigVMDispatch_SelectInt32.Result
    TargetMinimum=@n1.outputMinMax.X
    TargetMaximum=@n1.outputMinMax.Y
    SourceMinimum=@n1.rotationMinMax.X
    SourceMaximum=@n1.rotationMinMax.Y
   )
   Absolute = &RigVMFunction_MathDoubleAbs(Value=@Remap.Result)
   Multiply = &RigVMFunction_MathDoubleMul(
    A=&RigVMFunction_MathDoubleClamp(Value=RigVMDispatchNode(True=@Absolute.Result, Condition=@n1.absoluteValue, False=@Remap.Result) {template="DISPATCH_RigVMDispatch_If(in Condition,in True,in False,out Result)"}.Result).Result
    B=@n1.premult
   )
   > RigVMFunctionReturnNode(outputQuat=@n2.Result, outputWeight=@Multiply.Result)
  ]}
 RigVMFunctionLibrary {nodes [DeltaRotation v4 = RigVMCollapseNode(
    ExecuteContext=()
    Rotation Order=ZYX
    child=(Type=Bone,Name="None")
    outputMinMax=(X=0,Y=1)
    parent=(Type=Bone,Name="None")
    premult=1
    rotationMinMax=(X=0,Y=0)
    twistAxis=(X=1,Y=0,Z=0)
   )]}
 RigVMFunction_Sequence_1_ContainedGraph {nodes [
   n1 = RigVMFunctionEntryNode()
   RigVMFunction_Sequence = &RigVMFunction_Sequence(ExecuteContext=@n1.ExecuteContext)
   n2 = &RigVMFunction_Sequence(ExecuteContext=@RigVMFunction_Sequence.B)
   n3 = &RigVMFunction_Sequence(ExecuteContext=@n2.B)
   RigVMFunctionReturnNode(A=@RigVMFunction_Sequence.A, B=@n2.A, C=@n3.A, D=@n3.B)
  ]}
 RigVMModel {nodes [
   &RigUnit_ParentConstraint(
    AdvancedSettings=(InterpolationType=Average,RotationOrderForFilter=XZY)
    Child=(Type=Bone,Name="Shoulders_02_l")
    ExecutePin=()
    Filter=(TranslationFilter=(bX=True,bY=True,bZ=True),RotationFilter=(bX=false,bY=True,bZ=True),ScaleFilter=(bX=True,bY=True,bZ=True))
    Parents=((Item=(Type=Bone,Name="clavicle_l"),Weight=1))
    Weight=1
   )
   > &RigUnit_RotationConstraintLocalSpaceOffset(
    AdvancedSettings=(InterpolationType=Average,RotationOrderForFilter=XZY)
    Child=(Type=Bone,Name="Shoulders_02_l")
    Filter=(bX=false,bY=True,bZ=True)
    Parents=((Item=(Type=Bone,Name="upperarm_l"),Weight=1))
    Weight=1
   )
   n1 = &RigVMFunction_MathQuaternionToEuler(
    RotationOrder=ZYX
    Value=&RigUnit_GetTransform(Item=(Type=Bone,Name="upperarm_l"), Space=LocalSpace, bInitial=False).Transform.Rotation
   )
   > &RigUnit_ModifyTransforms(
    ItemToModify=((Item=(Type=Bone,Name="Shoulders_02_l"),Transform=(Rotation=(X=0,Y=0,Z=0,W=1),Translation=(X=0,Y=0,Z=0),Scale3D=(X=1,Y=1,Z=1))))
    Weight=1
    WeightMinimum=0
    ItemToModify.0.Transform.Rotation=&RigVMFunction_MathQuaternionFromEuler(Euler=(X=0,Y=0,Z=0), Euler.X=&RigVMFunction_MathDoubleMul(B=-0.4, A=@n1.Result.X).Result).Result
   )
   > n2 = &RigUnit_SphericalPoseReader(
    ActiveRegionScaleFactors=(PositiveWidth=1,NegativeWidth=1,PositiveHeight=1,NegativeHeight=1)
    ActiveRegionSize=0.3
    Debug=(bDrawDebug=True,bDraw2D=False,bDrawLocalAxes=False,DebugScale=15,DebugSegments=12,DebugThickness=0.05)
    DriverAxis=(X=0,Y=1,Z=0)
    DriverItem=(Type=Bone,Name="upperarm_l")
    FalloffRegionScaleFactors=(PositiveWidth=1,NegativeWidth=1,PositiveHeight=1,NegativeHeight=1)
    FalloffSize=0.45
    FlipHeightScaling=False
    FlipWidthScaling=False
    OptionalParentItem=(Type=Bone,Name="None")
    RotationOffset=(X=0,Y=130,Z=0)
   )
   > &RigUnit_ModifyTransforms(
    ItemToModify=((Item=(Type=Bone,Name="Shoulders_02_l"),Transform=(Rotation=(X=0,Y=0,Z=0,W=1),Translation=(X=-0.2,Y=0,Z=0.2),Scale3D=(X=1,Y=1,Z=1))))
    WeightMinimum=0
    Weight=@n2.OutputParam
   )
   > &RigUnit_ModifyTransforms(
    ItemToModify=((Item=(Type=Bone,Name="Shoulders_01_l"),Transform=(Rotation=(X=-0.0436,Y=0.0019,Z=-0.0436,W=0.9981),Translation=(X=0,Y=0.2,Z=0.3),Scale3D=(X=0.9,Y=1,Z=1))))
    WeightMinimum=0
    Weight=@n2.OutputParam
   )
   n3 = &RigUnit_PrepareForExecution()
   RigUnit_GetTransform = &RigUnit_GetTransform(Item=(Type=Bone,Name="upperarm_r"), bInitial=False)
   &RigUnit_HierarchyAddNull(
    Name=apperarm_r_null
    Parent=(Type=Bone,Name="clavicle_r")
    Space=GlobalSpace
    ExecutePin=@n3.ExecutePin
    Transform=@RigUnit_GetTransform.Transform
   )
   > &RigUnit_HierarchyAddNull(
    Name=apperarm_l_null
    Parent=(Type=Bone,Name="clavicle_l")
    Space=GlobalSpace
    Transform=&RigUnit_GetTransform(Item=(Type=Bone,Name="upperarm_l"), bInitial=False).Transform
   )
   RigUnit_BeginExecution = &RigUnit_BeginExecution()
   /Script/FortniteGame.RigUnit_FortDeformRig(
    Bool=(bUpperArm=True,bElbow=True,bWrist=True,bGroin=True,bGlute=True,bKnee=True,bPec=True,bBicep=True,bLat=True)
    Curve="(bLUpperArmUpAlpha=False,bRUpperArmUpAlpha=False,bLUpperArmFrAlpha=False,bRUpperArmFrAlpha=False,bLUpperArmBkAlpha=False,bRUpperArmBkAlpha=False,bLWristUpAlpha=False,bRWristUpAlpha=False,bLWristDnAlpha=False,bRWristDnAlpha=False,bLKneeAlpha=False,bRKneeAlp ...(truncated)"
    Debug=(bUpperArmDebug=False,bWristDebug=False,bLatDebug=False,bGroinGluteDebug=False)
    Value="(UpperArmUp=(Rotation=(X=0.0,Y=-0.0,Z=0.0,W=1.0),Translation=(X=0.0,Y=0.0,Z=8.0),Scale3D=(X=0.70,Y=1.0,Z=1.0)),UpperArmFwdFr=(Rotation=(X=0.0,Y=-0.0,Z=0.0,W=1.0),Translation=(X=0.0,Y=10.0,Z=0.0),Scale3D=(X=0.50,Y=1.0,Z=1.0)),UpperArmFwdBk=(Rotation=(X=0.0, ...(truncated)"
    ExecutePin=@RigUnit_BeginExecution.ExecutePin
   ) {is=struct RigVMUnitNode}
   > RigVMFunction_Sequence = RigVMAggregateNode()
   &RigUnit_ModifyTransforms(
    ItemToModify=((Item=(Type=Bone,Name="clavicle_l"),Transform=(Rotation=(X=0,Y=0,Z=0,W=1),Translation=(X=0,Y=0,Z=0),Scale3D=(X=1,Y=1,Z=1))))
    Weight=1
    WeightMinimum=0
    ExecutePin=@RigVMFunction_Sequence.A
   )
   > n4 = &RigUnit_SphericalPoseReader(
    ActiveRegionScaleFactors=(PositiveWidth=1,NegativeWidth=0.5,PositiveHeight=1,NegativeHeight=1)
    ActiveRegionSize=0.14
    Debug=(bDrawDebug=false,bDraw2D=False,bDrawLocalAxes=False,DebugScale=15,DebugSegments=12,DebugThickness=0.02)
    DriverAxis=(X=-1,Y=0,Z=0)
    DriverItem=(Type=Bone,Name="clavicle_r")
    FalloffRegionScaleFactors=(PositiveWidth=1,NegativeWidth=1,PositiveHeight=1,NegativeHeight=1)
    FalloffSize=0.38
    FlipHeightScaling=False
    FlipWidthScaling=False
    OptionalParentItem=(Type=Bone,Name="None")
    RotationOffset=(X=180,Y=-20,Z=0)
   )
   > &RigUnit_ModifyTransforms(
    ItemToModify=((Item=(Type=Bone,Name="Shoulders_02_r"),Transform=(Rotation=(X=0.1299,Y=0.0972,Z=0.0128,W=0.9867),Translation=(X=-3.5768,Y=0.5977,Z=-2.7857),Scale3D=(X=1,Y=1,Z=1))))
    WeightMinimum=0
    Weight=@n4.OutputParam
   )
   > &RigUnit_ModifyTransforms(
    ItemToModify=((Item=(Type=Bone,Name="Shoulders_01_r"),Transform=(Rotation=(X=0.2587,Y=-0.0237,Z=-0.0064,W=0.9656),Translation=(X=-4.7222,Y=3.0284,Z=0.6629),Scale3D=(X=0.9,Y=1,Z=1))))
    WeightMinimum=0
    Weight=@n4.OutputParam
   )
   > n5 = &RigUnit_SphericalPoseReader(
    ActiveRegionScaleFactors=(PositiveWidth=0.5,NegativeWidth=1,PositiveHeight=1,NegativeHeight=1)
    ActiveRegionSize=0.14
    Debug=(bDrawDebug=false,bDraw2D=false,bDrawLocalAxes=False,DebugScale=15,DebugSegments=12,DebugThickness=0.02)
    DriverAxis=(X=1,Y=0,Z=0)
    DriverItem=(Type=Bone,Name="clavicle_l")
    FalloffRegionScaleFactors=(PositiveWidth=1,NegativeWidth=1,PositiveHeight=1,NegativeHeight=1)
    FalloffSize=0.38
    FlipHeightScaling=False
    FlipWidthScaling=False
    OptionalParentItem=(Type=Bone,Name="None")
    RotationOffset=(X=0,Y=-20,Z=0)
   )
   > &RigUnit_ModifyTransforms(
    ItemToModify=((Item=(Type=Bone,Name="Shoulders_02_l"),Transform=(Rotation=(X=0.1322,Y=0.094,Z=0.037,W=0.9861),Translation=(X=4.1625,Y=-0.9816,Z=1.9295),Scale3D=(X=1,Y=1,Z=1))))
    WeightMinimum=0
    Weight=@n5.OutputParam
   )
   > &RigUnit_ModifyTransforms(
    ItemToModify=((Item=(Type=Bone,Name="Shoulders_01_l"),Transform=(Rotation=(X=0.3296,Y=-0.016,Z=-0.0278,W=0.9436),Translation=(X=4.7036,Y=-4.35,Z=-0.8743),Scale3D=(X=0.9,Y=1,Z=1))))
    WeightMinimum=0
    Weight=@n5.OutputParam
   )
   > &RigUnit_ModifyTransforms(
    ItemToModify=((Item=(Type=Bone,Name="Shoulder_twist_l"),Transform=(Rotation=(X=0,Y=0,Z=0,W=1),Translation=(X=5.4417,Y=0,Z=-5.3579),Scale3D=(X=0.65,Y=0.65,Z=0.65))))
    Weight=1
    WeightMinimum=0
   )
   > &RigUnit_ModifyTransforms(
    ItemToModify=((Item=(Type=Bone,Name="Shoulder_twist_r"),Transform=(Rotation=(X=0,Y=0,Z=0,W=-1),Translation=(X=-5.4417,Y=0,Z=5.3579),Scale3D=(X=0.65,Y=0.65,Z=0.65))))
    Weight=1
    WeightMinimum=0
   )
   &RigUnit_ModifyTransforms(
    ItemToModify=((Item=(Type=Bone,Name="upperarm_r"),Transform=(Rotation=(X=0,Y=0,Z=0,W=1),Translation=(X=0,Y=0,Z=0),Scale3D=(X=1,Y=1,Z=1))))
    Weight=1
    WeightMinimum=0
    ExecutePin=@RigVMFunction_Sequence.B
   )
   > &RigUnit_ParentConstraint(
    AdvancedSettings=(InterpolationType=Average,RotationOrderForFilter=XZY)
    Child=(Type=Bone,Name="Shoulders_02_r")
    Filter=(TranslationFilter=(bX=True,bY=True,bZ=True),RotationFilter=(bX=true,bY=True,bZ=True),ScaleFilter=(bX=True,bY=True,bZ=True))
    Parents=((Item=(Type=Bone,Name="clavicle_r"),Weight=1))
    Weight=1
   )
   > &RigUnit_RotationConstraintLocalSpaceOffset(
    AdvancedSettings=(InterpolationType=Average,RotationOrderForFilter=XZY)
    Child=(Type=Bone,Name="Shoulders_02_r")
    Filter=(bX=true,bY=True,bZ=True)
    Parents=((Item=(Type=Bone,Name="upperarm_r"),Weight=1))
    Weight=1
   )
   > DeltaRotation v4_2 = RigVMFunctionReferenceNode(
    Rotation Order=ZYX
    child=(Type=Bone,Name="upperarm_r")
    onlyTwist=true
    outputMinMax=(X=0,Y=1)
    parent=(Type=Null,Name="apperarm_r_null")
    premult=1
    rotationMinMax=(X=0,Y=45)
    twistAxis=(X=1,Y=0,Z=0)
   )
   > &RigUnit_ModifyTransforms(
    ItemToModify=((Item=(Type=Bone,Name="Shoulders_02_r"),Transform=(Rotation=(X=-0.3827,Y=0,Z=0,W=0.9239),Translation=(X=0,Y=0.5,Z=-0.7),Scale3D=(X=1,Y=1,Z=1))))
    WeightMinimum=0
    Weight=@"DeltaRotation v4_2".outputWeight
   )
   > DeltaRotation v4 = RigVMFunctionReferenceNode(
    Rotation Order=ZYX
    child=(Type=Bone,Name="upperarm_r")
    onlyTwist=true
    outputMinMax=(X=0,Y=1)
    parent=(Type=Null,Name="apperarm_r_null")
    premult=1
    rotationMinMax=(X=0,Y=-30)
    twistAxis=(X=1,Y=0,Z=0)
   )
   > &RigUnit_ModifyTransforms(
    ItemToModify=((Item=(Type=Bone,Name="Shoulders_02_r"),Transform=(Rotation=(X=0.2588,Y=0,Z=0,W=0.9659),Translation=(X=0,Y=0.5,Z=-0.7),Scale3D=(X=1,Y=1,Z=1))))
    WeightMinimum=0
    Weight=@"DeltaRotation v4".outputWeight
   )
   > &RigUnit_ModifyTransforms(
    ItemToModify=((Item=(Type=Bone,Name="Shoulders_01_r"),Transform=(Rotation=(X=-0.0436,Y=0,Z=0,W=0.999),Translation=(X=0,Y=0,Z=0),Scale3D=(X=0.9,Y=1,Z=1))))
    WeightMinimum=0
    Weight=@"DeltaRotation v4".outputWeight
   )
   > n6 = &RigUnit_SphericalPoseReader(
    ActiveRegionScaleFactors=(PositiveWidth=1,NegativeWidth=1,PositiveHeight=1,NegativeHeight=1)
    ActiveRegionSize=0.1
    Debug=(bDrawDebug=True,bDraw2D=False,bDrawLocalAxes=False,DebugScale=15,DebugSegments=12,DebugThickness=0.05)
    DriverAxis=(X=-1,Y=0,Z=0)
    DriverItem=(Type=Bone,Name="upperarm_r")
    FalloffRegionScaleFactors=(PositiveWidth=1,NegativeWidth=1,PositiveHeight=1.5,NegativeHeight=1.5)
    FalloffSize=0.275
    FlipHeightScaling=False
    FlipWidthScaling=False
    OptionalParentItem=(Type=Bone,Name="None")
    RotationOffset=(X=0,Y=40,Z=0)
   )
   > &RigUnit_ModifyTransforms(
    ItemToModify=((Item=(Type=Bone,Name="Shoulders_01_r"),Transform=(Rotation=(X=0,Y=0.0736,Z=0,W=-0.9973),Translation=(X=1,Y=0.2549,Z=-0.1285),Scale3D=(X=0.92,Y=1,Z=1))))
    WeightMinimum=0
    Weight=@n6.OutputParam
   )
   > &RigUnit_ModifyTransforms(
    ItemToModify=((Item=(Type=Bone,Name="Shoulders_02_r"),Transform=(Rotation=(X=0,Y=0,Z=0,W=-1),Translation=(X=0.4424,Y=0,Z=-0.7),Scale3D=(X=1,Y=1,Z=1))))
    WeightMinimum=0
    Weight=@n6.OutputParam
   )
   > n7 = &RigUnit_SphericalPoseReader(
    ActiveRegionScaleFactors=(PositiveWidth=1,NegativeWidth=1,PositiveHeight=1.5,NegativeHeight=1)
    ActiveRegionSize=0.08
    Debug=(bDrawDebug=True,bDraw2D=False,bDrawLocalAxes=False,DebugScale=15,DebugSegments=12,DebugThickness=0.05)
    DriverAxis=(X=-1,Y=0,Z=0)
    DriverItem=(Type=Bone,Name="upperarm_r")
    FalloffRegionScaleFactors=(PositiveWidth=1,NegativeWidth=1,PositiveHeight=1,NegativeHeight=1)
    FalloffSize=0.495
    FlipHeightScaling=False
    FlipWidthScaling=False
    OptionalParentItem=(Type=Bone,Name="None")
    RotationOffset=(X=180,Y=0,Z=0)
   )
   > &RigUnit_ModifyTransforms(
    ItemToModify=((Item=(Type=Bone,Name="Shoulders_02_r"),Transform=(Rotation=(X=0.0852,Y=0.0711,Z=0.0188,W=0.9936),Translation=(X=-8,Y=3,Z=-4),Scale3D=(X=1,Y=1,Z=1))))
    WeightMinimum=0
    Weight=@n7.OutputParam
   )
   > &RigUnit_ModifyTransforms(
    ItemToModify=((Item=(Type=Bone,Name="Shoulders_01_r"),Transform=(Rotation=(X=0.1709,Y=0.0245,Z=0.0018,W=0.985),Translation=(X=1,Y=2.0148,Z=-1.1744),Scale3D=(X=0.9,Y=1,Z=1))))
    WeightMinimum=0
    Weight=@n7.OutputParam
   )
   > n8 = &RigUnit_SphericalPoseReader(
    ActiveRegionScaleFactors=(PositiveWidth=1,NegativeWidth=1,PositiveHeight=1,NegativeHeight=1)
    ActiveRegionSize=0.1
    Debug=(bDrawDebug=True,bDraw2D=False,bDrawLocalAxes=False,DebugScale=15,DebugSegments=12,DebugThickness=0.05)
    DriverAxis=(X=-1,Y=0,Z=0)
    DriverItem=(Type=Bone,Name="upperarm_r")
    FalloffRegionScaleFactors=(PositiveWidth=1,NegativeWidth=1,PositiveHeight=1,NegativeHeight=1)
    FalloffSize=0.495
    FlipHeightScaling=False
    FlipWidthScaling=False
    OptionalParentItem=(Type=Bone,Name="None")
    RotationOffset=(X=-90,Y=0,Z=0)
   )
   > &RigUnit_ModifyTransforms(
    ItemToModify=((Item=(Type=Bone,Name="Shoulders_02_r"),Transform=(Rotation=(X=0.2588,Y=0,Z=0,W=0.9659),Translation=(X=0.7,Y=1,Z=-1.7),Scale3D=(X=1,Y=1,Z=1))))
    WeightMinimum=0
    Weight=@n8.OutputParam
   )
   > &RigUnit_ModifyTransforms(
    ItemToModify=((Item=(Type=Bone,Name="Shoulders_01_r"),Transform=(Rotation=(X=0.1899,Y=-0.0023,Z=0.1014,W=0.9766),Translation=(X=1.2558,Y=3.3263,Z=-0.975),Scale3D=(X=0.95,Y=1,Z=1))))
    WeightMinimum=0
    Weight=@n8.OutputParam
   )
   > n9 = &RigUnit_SphericalPoseReader(
    ActiveRegionScaleFactors=(PositiveWidth=1,NegativeWidth=1,PositiveHeight=1,NegativeHeight=1)
    ActiveRegionSize=0.1
    Debug=(bDrawDebug=True,bDraw2D=False,bDrawLocalAxes=False,DebugScale=15,DebugSegments=12,DebugThickness=0.05)
    DriverAxis=(X=-1,Y=0,Z=0)
    DriverItem=(Type=Bone,Name="upperarm_r")
    FalloffRegionScaleFactors=(PositiveWidth=1,NegativeWidth=1,PositiveHeight=1,NegativeHeight=1)
    FalloffSize=0.495
    FlipHeightScaling=False
    FlipWidthScaling=False
    OptionalParentItem=(Type=Bone,Name="None")
    RotationOffset=(X=90,Y=0,Z=0)
   )
   > &RigUnit_ModifyTransforms(
    ItemToModify=((Item=(Type=Bone,Name="Shoulders_02_r"),Transform=(Rotation=(X=-0.3369,Y=0,Z=0,W=0.9415),Translation=(X=-1.1252,Y=-0.5027,Z=-1.0045),Scale3D=(X=1,Y=1,Z=1))))
    WeightMinimum=0
    Weight=@n9.OutputParam
   )
   > &RigUnit_ModifyTransforms(
    ItemToModify=((Item=(Type=Bone,Name="Shoulders_01_r"),Transform=(Rotation=(X=0,Y=0,Z=-0.1305,W=0.9914),Translation=(X=0.9913,Y=-0.1814,Z=-0.5594),Scale3D=(X=0.9,Y=1,Z=1))))
    WeightMinimum=0
    Weight=@n9.OutputParam
   )
   &RigUnit_ModifyTransforms(
    ItemToModify=((Item=(Type=Bone,Name="upperarm_l"),Transform=(Rotation=(X=0,Y=0,Z=0,W=1),Translation=(X=0,Y=0,Z=0),Scale3D=(X=1,Y=1,Z=1))))
    Weight=1
    WeightMinimum=0
    ExecutePin=@RigVMFunction_Sequence.C
   )
   > &RigUnit_ParentConstraint(
    AdvancedSettings=(InterpolationType=Average,RotationOrderForFilter=XZY)
    Child=(Type=Bone,Name="Shoulders_02_l")
    Filter=(TranslationFilter=(bX=True,bY=True,bZ=True),RotationFilter=(bX=false,bY=True,bZ=True),ScaleFilter=(bX=True,bY=True,bZ=True))
    Parents=((Item=(Type=Bone,Name="clavicle_l"),Weight=1))
    Weight=1
   )
   > &RigUnit_RotationConstraintLocalSpaceOffset(
    AdvancedSettings=(InterpolationType=Average,RotationOrderForFilter=XZY)
    Child=(Type=Bone,Name="Shoulders_02_l")
    Filter=(bX=false,bY=True,bZ=True)
    Parents=((Item=(Type=Bone,Name="upperarm_l"),Weight=1))
    Weight=1
   )
   > DeltaRotation v4_3 = RigVMFunctionReferenceNode(
    Rotation Order=ZYX
    child=(Type=Bone,Name="upperarm_l")
    onlyTwist=true
    outputMinMax=(X=0,Y=1)
    parent=(Type=Null,Name="apperarm_l_null")
    premult=1
    rotationMinMax=(X=0,Y=45)
    twistAxis=(X=1,Y=0,Z=0)
   )
   > &RigUnit_ModifyTransforms(
    ItemToModify=((Item=(Type=Bone,Name="Shoulders_02_l"),Transform=(Rotation=(X=-0.3827,Y=0,Z=0,W=0.9239),Translation=(X=0,Y=-0.6,Z=1),Scale3D=(X=1,Y=1,Z=1))))
    WeightMinimum=0
    Weight=@"DeltaRotation v4_3".outputWeight
   )
   > DeltaRotation v4_1 = RigVMFunctionReferenceNode(
    Rotation Order=ZYX
    child=(Type=Bone,Name="upperarm_l")
    onlyTwist=true
    outputMinMax=(X=0,Y=1)
    parent=(Type=Null,Name="apperarm_l_null")
    premult=1
    rotationMinMax=(X=0,Y=-45)
    twistAxis=(X=1,Y=0,Z=0)
   )
   > &RigUnit_ModifyTransforms(
    ItemToModify=((Item=(Type=Bone,Name="Shoulders_02_l"),Transform=(Rotation=(X=0.3827,Y=0,Z=0,W=0.9239),Translation=(X=0,Y=-0.6,Z=1),Scale3D=(X=1,Y=1,Z=1))))
    WeightMinimum=0
    Weight=@"DeltaRotation v4_1".outputWeight
   )
   > &RigUnit_ModifyTransforms(
    ItemToModify=((Item=(Type=Bone,Name="Shoulders_01_l"),Transform=(Rotation=(X=-0.0436,Y=0.0019,Z=-0.0436,W=0.9981),Translation=(X=0,Y=0.2,Z=0.3),Scale3D=(X=0.9,Y=1,Z=1))))
    WeightMinimum=0
    Weight=@"DeltaRotation v4_1".outputWeight
   )
   > n10 = &RigUnit_SphericalPoseReader(
    ActiveRegionScaleFactors=(PositiveWidth=1,NegativeWidth=1,PositiveHeight=1,NegativeHeight=1)
    ActiveRegionSize=0.1
    Debug=(bDrawDebug=True,bDraw2D=False,bDrawLocalAxes=False,DebugScale=15,DebugSegments=12,DebugThickness=0.05)
    DriverAxis=(X=1,Y=0,Z=0)
    DriverItem=(Type=Bone,Name="upperarm_l")
    FalloffRegionScaleFactors=(PositiveWidth=1,NegativeWidth=1,PositiveHeight=1.5,NegativeHeight=1.5)
    FalloffSize=0.275
    FlipHeightScaling=False
    FlipWidthScaling=False
    OptionalParentItem=(Type=Bone,Name="None")
    RotationOffset=(X=0,Y=-140,Z=0)
   )
   > &RigUnit_ModifyTransforms(
    ItemToModify=((Item=(Type=Bone,Name="Shoulders_01_l"),Transform=(Rotation=(X=0,Y=0.0736,Z=0,W=-0.9973),Translation=(X=-0.7603,Y=-0.2549,Z=0.1285),Scale3D=(X=0.93,Y=1,Z=1))))
    WeightMinimum=0
    Weight=@n10.OutputParam
   )
   > &RigUnit_ModifyTransforms(
    ItemToModify=((Item=(Type=Bone,Name="Shoulders_02_l"),Transform=(Rotation=(X=0,Y=0,Z=0,W=-1),Translation=(X=-0.4424,Y=0,Z=0.5033),Scale3D=(X=1,Y=1,Z=1))))
    WeightMinimum=0
    Weight=@n10.OutputParam
   )
   > n11 = &RigUnit_SphericalPoseReader(
    ActiveRegionScaleFactors=(PositiveWidth=1,NegativeWidth=1,PositiveHeight=1.5,NegativeHeight=1)
    ActiveRegionSize=0.08
    Debug=(bDrawDebug=True,bDraw2D=False,bDrawLocalAxes=False,DebugScale=15,DebugSegments=12,DebugThickness=0.05)
    DriverAxis=(X=1,Y=0,Z=0)
    DriverItem=(Type=Bone,Name="upperarm_l")
    FalloffRegionScaleFactors=(PositiveWidth=1,NegativeWidth=1,PositiveHeight=1,NegativeHeight=1)
    FalloffSize=0.495
    FlipHeightScaling=False
    FlipWidthScaling=False
    OptionalParentItem=(Type=Bone,Name="None")
    RotationOffset=(X=0,Y=0,Z=0)
   )
   > &RigUnit_ModifyTransforms(
    ItemToModify=((Item=(Type=Bone,Name="Shoulders_02_l"),Transform=(Rotation=(X=0.1705,Y=0.0725,Z=0.0126,W=0.9826),Translation=(X=7.5442,Y=-0.7315,Z=2.5427),Scale3D=(X=1,Y=1,Z=1))))
    WeightMinimum=0
    Weight=@n11.OutputParam
   )
   > &RigUnit_ModifyTransforms(
    ItemToModify=((Item=(Type=Bone,Name="Shoulders_01_l"),Transform=(Rotation=(X=0.1709,Y=0.0245,Z=0.0018,W=0.985),Translation=(X=-1,Y=-2.0148,Z=1.1744),Scale3D=(X=0.9,Y=1,Z=1))))
    WeightMinimum=0
    Weight=@n11.OutputParam
   )
   > n12 = &RigUnit_SphericalPoseReader(
    ActiveRegionScaleFactors=(PositiveWidth=1,NegativeWidth=1,PositiveHeight=1,NegativeHeight=1)
    ActiveRegionSize=0.1
    Debug=(bDrawDebug=True,bDraw2D=False,bDrawLocalAxes=False,DebugScale=15,DebugSegments=12,DebugThickness=0.05)
    DriverAxis=(X=1,Y=0,Z=0)
    DriverItem=(Type=Bone,Name="upperarm_l")
    FalloffRegionScaleFactors=(PositiveWidth=1,NegativeWidth=1,PositiveHeight=1,NegativeHeight=1)
    FalloffSize=0.495
    FlipHeightScaling=False
    FlipWidthScaling=False
    OptionalParentItem=(Type=Bone,Name="None")
    RotationOffset=(X=90,Y=0,Z=0)
   )
   > &RigUnit_ModifyTransforms(
    ItemToModify=((Item=(Type=Bone,Name="Shoulders_02_l"),Transform=(Rotation=(X=0.2903,Y=0,Z=0,W=0.9569),Translation=(X=-0.6376,Y=-1.853,Z=2.0503),Scale3D=(X=1,Y=1,Z=1))))
    WeightMinimum=0
    Weight=@n12.OutputParam
   )
   > &RigUnit_ModifyTransforms(
    ItemToModify=((Item=(Type=Bone,Name="Shoulders_01_l"),Transform=(Rotation=(X=0.1899,Y=-0.0023,Z=0.1014,W=0.9766),Translation=(X=-1.2558,Y=-3.3263,Z=0.975),Scale3D=(X=0.95,Y=1,Z=1))))
    WeightMinimum=0
    Weight=@n12.OutputParam
   )
   > n13 = &RigUnit_SphericalPoseReader(
    ActiveRegionScaleFactors=(PositiveWidth=1,NegativeWidth=1,PositiveHeight=1,NegativeHeight=1)
    ActiveRegionSize=0.1
    Debug=(bDrawDebug=True,bDraw2D=False,bDrawLocalAxes=False,DebugScale=15,DebugSegments=12,DebugThickness=0.05)
    DriverAxis=(X=1,Y=0,Z=0)
    DriverItem=(Type=Bone,Name="upperarm_l")
    FalloffRegionScaleFactors=(PositiveWidth=1,NegativeWidth=1,PositiveHeight=1,NegativeHeight=1)
    FalloffSize=0.495
    FlipHeightScaling=False
    FlipWidthScaling=False
    OptionalParentItem=(Type=Bone,Name="None")
    RotationOffset=(X=-90,Y=0,Z=0)
   )
   > &RigUnit_ModifyTransforms(
    ItemToModify=((Item=(Type=Bone,Name="Shoulders_02_l"),Transform=(Rotation=(X=-0.3369,Y=0,Z=0,W=0.9415),Translation=(X=1.1252,Y=0.5027,Z=1.0045),Scale3D=(X=1,Y=1,Z=1))))
    WeightMinimum=0
    Weight=@n13.OutputParam
   )
   > &RigUnit_ModifyTransforms(
    ItemToModify=((Item=(Type=Bone,Name="Shoulders_01_l"),Transform=(Rotation=(X=0,Y=0,Z=-0.1305,W=0.9914),Translation=(X=-0.9913,Y=0.1814,Z=0.5594),Scale3D=(X=0.9,Y=1,Z=1))))
    WeightMinimum=0
    Weight=@n13.OutputParam
   )
   n14 = &RigUnit_SphericalPoseReader(
    ActiveRegionScaleFactors=(PositiveWidth=1,NegativeWidth=1,PositiveHeight=1,NegativeHeight=1)
    ActiveRegionSize=0.06
    Debug=(bDrawDebug=true,bDraw2D=False,bDrawLocalAxes=False,DebugScale=20,DebugSegments=12,DebugThickness=0.05)
    DriverAxis=(X=1,Y=0,Z=0)
    DriverItem=(Type=Bone,Name="calf_r")
    FalloffRegionScaleFactors=(PositiveWidth=1,NegativeWidth=1,PositiveHeight=1,NegativeHeight=1)
    FalloffSize=0.49
    FlipHeightScaling=False
    FlipWidthScaling=False
    OptionalParentItem=(Type=Bone,Name="None")
    RotationOffset=(X=90,Y=0,Z=30)
    ExecutePin=@RigVMFunction_Sequence.D
   )
   > &RigUnit_ModifyTransforms(
    ItemToModify=((Item=(Type=Bone,Name="Kneepad_r"),Transform=(Rotation=(X=0,Y=0,Z=-0.1951,W=0.9808),Translation=(X=-1.9264,Y=1.2942,Z=-0.0723),Scale3D=(X=1,Y=1,Z=1))))
    WeightMinimum=0
    Weight=@n14.OutputParam
   )
   > n15 = &RigUnit_SphericalPoseReader(
    ActiveRegionScaleFactors=(PositiveWidth=1,NegativeWidth=1,PositiveHeight=1,NegativeHeight=1)
    ActiveRegionSize=0.06
    Debug=(bDrawDebug=true,bDraw2D=False,bDrawLocalAxes=False,DebugScale=20,DebugSegments=12,DebugThickness=0.05)
    DriverAxis=(X=-1,Y=0,Z=0)
    DriverItem=(Type=Bone,Name="calf_l")
    FalloffRegionScaleFactors=(PositiveWidth=1,NegativeWidth=1,PositiveHeight=1,NegativeHeight=1)
    FalloffSize=0.49
    FlipHeightScaling=False
    FlipWidthScaling=False
    OptionalParentItem=(Type=Bone,Name="None")
    RotationOffset=(X=-90,Y=0,Z=30)
   )
   > &RigUnit_ModifyTransforms(
    ItemToModify=((Item=(Type=Bone,Name="Kneepad_l"),Transform=(Rotation=(X=0,Y=0,Z=-0.171,W=0.9853),Translation=(X=1.9264,Y=-1.2942,Z=0.0723),Scale3D=(X=1,Y=1,Z=1))))
    WeightMinimum=0
    Weight=@n15.OutputParam
   )
   > n16 = &RigUnit_SphericalPoseReader(
    ActiveRegionScaleFactors=(PositiveWidth=1,NegativeWidth=1,PositiveHeight=2,NegativeHeight=2)
    ActiveRegionSize=0.06
    Debug=(bDrawDebug=true,bDraw2D=False,bDrawLocalAxes=False,DebugScale=20,DebugSegments=12,DebugThickness=0.05)
    DriverAxis=(X=0,Y=-1,Z=0)
    DriverItem=(Type=Bone,Name="foot_r")
    FalloffRegionScaleFactors=(PositiveWidth=1,NegativeWidth=1,PositiveHeight=1,NegativeHeight=1)
    FalloffSize=0.22
    FlipHeightScaling=False
    FlipWidthScaling=False
    OptionalParentItem=(Type=Bone,Name="None")
    RotationOffset=(X=-90,Y=0,Z=-40)
   )
   > &RigUnit_ModifyTransforms(
    ItemToModify=((Item=(Type=Bone,Name="boots_cor_01_r1"),Transform=(Rotation=(X=0,Y=0,Z=-0.098,W=-0.9952),Translation=(X=0.0612,Y=-2.5621,Z=0),Scale3D=(X=1,Y=1,Z=1))))
    WeightMinimum=0
    Weight=@n16.OutputParam
   )
   > n17 = &RigUnit_SphericalPoseReader(
    ActiveRegionScaleFactors=(PositiveWidth=1,NegativeWidth=1,PositiveHeight=2,NegativeHeight=2)
    ActiveRegionSize=0.06
    Debug=(bDrawDebug=true,bDraw2D=False,bDrawLocalAxes=False,DebugScale=20,DebugSegments=12,DebugThickness=0.05)
    DriverAxis=(X=0,Y=1,Z=0)
    DriverItem=(Type=Bone,Name="foot_l")
    FalloffRegionScaleFactors=(PositiveWidth=1,NegativeWidth=1,PositiveHeight=1,NegativeHeight=1)
    FalloffSize=0.22
    FlipHeightScaling=False
    FlipWidthScaling=False
    OptionalParentItem=(Type=Bone,Name="None")
    RotationOffset=(X=90,Y=0,Z=-40)
   )
   > &RigUnit_ModifyTransforms(
    ItemToModify=((Item=(Type=Bone,Name="boots_cor_01_l1"),Transform=(Rotation=(X=0,Y=0,Z=-0.098,W=-0.9952),Translation=(X=-0.0612,Y=2.5621,Z=0),Scale3D=(X=1,Y=1,Z=1))))
    WeightMinimum=0
    Weight=@n17.OutputParam
   )
   > n18 = &RigUnit_SphericalPoseReader(
    ActiveRegionScaleFactors=(PositiveWidth=1,NegativeWidth=1,PositiveHeight=2,NegativeHeight=2)
    ActiveRegionSize=0.06
    Debug=(bDrawDebug=true,bDraw2D=False,bDrawLocalAxes=False,DebugScale=20,DebugSegments=12,DebugThickness=0.05)
    DriverAxis=(X=0,Y=1,Z=0)
    DriverItem=(Type=Bone,Name="foot_r")
    FalloffRegionScaleFactors=(PositiveWidth=1,NegativeWidth=1,PositiveHeight=1,NegativeHeight=1)
    FalloffSize=0.22
    FlipHeightScaling=False
    FlipWidthScaling=False
    OptionalParentItem=(Type=Bone,Name="None")
    RotationOffset=(X=90,Y=0,Z=75)
   )
   > &RigUnit_ModifyTransforms(
    ItemToModify=((Item=(Type=Bone,Name="boots_cor_01_r"),Transform=(Rotation=(X=0,Y=0,Z=0.0491,W=0.9988),Translation=(X=-4.791,Y=-0.9771,Z=0),Scale3D=(X=1,Y=1,Z=1))))
    WeightMinimum=0
    Weight=@n18.OutputParam
   )
   > n19 = &RigUnit_SphericalPoseReader(
    ActiveRegionScaleFactors=(PositiveWidth=1,NegativeWidth=1,PositiveHeight=2,NegativeHeight=2)
    ActiveRegionSize=0.06
    Debug=(bDrawDebug=true,bDraw2D=False,bDrawLocalAxes=False,DebugScale=20,DebugSegments=12,DebugThickness=0.05)
    DriverAxis=(X=0,Y=-1,Z=0)
    DriverItem=(Type=Bone,Name="foot_l")
    FalloffRegionScaleFactors=(PositiveWidth=1,NegativeWidth=1,PositiveHeight=1,NegativeHeight=1)
    FalloffSize=0.22
    FlipHeightScaling=False
    FlipWidthScaling=False
    OptionalParentItem=(Type=Bone,Name="None")
    RotationOffset=(X=-90,Y=0,Z=75)
   )
   > &RigUnit_ModifyTransforms(
    ItemToModify=((Item=(Type=Bone,Name="boots_cor_01_l"),Transform=(Rotation=(X=0,Y=0,Z=0.0491,W=0.9988),Translation=(X=4.791,Y=-0.9771,Z=0),Scale3D=(X=1,Y=1,Z=1))))
    WeightMinimum=0
    Weight=@n19.OutputParam
   )
   n20 = &RigUnit_SphericalPoseReader(
    ActiveRegionScaleFactors=(PositiveWidth=1,NegativeWidth=1.4,PositiveHeight=1,NegativeHeight=1)
    ActiveRegionSize=0.3
    Debug=(bDrawDebug=True,bDraw2D=False,bDrawLocalAxes=False,DebugScale=15,DebugSegments=12,DebugThickness=0.05)
    DriverAxis=(X=0,Y=1,Z=0)
    DriverItem=(Type=Bone,Name="upperarm_l")
    ExecutePin=()
    FalloffRegionScaleFactors=(PositiveWidth=1,NegativeWidth=2,PositiveHeight=1,NegativeHeight=1)
    FalloffSize=0.49
    FlipHeightScaling=False
    FlipWidthScaling=False
    OptionalParentItem=(Type=Bone,Name="None")
    RotationOffset=(X=0,Y=130,Z=0)
   )
   &RigUnit_ModifyTransforms(
    ExecutePin=()
    ItemToModify=((Item=(Type=Bone,Name="Shoulders_02_l"),Transform=(Rotation=(X=0,Y=0,Z=0,W=1),Translation=(X=0.7702,Y=-0.8,Z=1),Scale3D=(X=1,Y=1,Z=1))))
    WeightMinimum=0
    Weight=@n20.OutputParam
   )
   n21 = &RigUnit_SphericalPoseReader(
    ActiveRegionScaleFactors=(PositiveWidth=1,NegativeWidth=1.4,PositiveHeight=1,NegativeHeight=1)
    ActiveRegionSize=0.3
    Debug=(bDrawDebug=True,bDraw2D=False,bDrawLocalAxes=False,DebugScale=15,DebugSegments=12,DebugThickness=0.05)
    DriverAxis=(X=0,Y=-1,Z=0)
    DriverItem=(Type=Bone,Name="upperarm_r")
    ExecutePin=()
    FalloffRegionScaleFactors=(PositiveWidth=1,NegativeWidth=2,PositiveHeight=1,NegativeHeight=1)
    FalloffSize=0.49
    FlipHeightScaling=False
    FlipWidthScaling=False
    OptionalParentItem=(Type=Bone,Name="None")
    RotationOffset=(X=0,Y=-50,Z=0)
   )
   &RigUnit_ModifyTransforms(
    ExecutePin=()
    ItemToModify=((Item=(Type=Bone,Name="Shoulders_02_r"),Transform=(Rotation=(X=0.4222,Y=0.0184,Z=0.0395,W=0.9054),Translation=(X=0.8,Y=0.3,Z=0),Scale3D=(X=1,Y=1,Z=1))))
    WeightMinimum=0
    Weight=@n21.OutputParam
   )
  ]}
]
hierarchy [
 root {loc=<0 0 0>, rot=<0 0 0 1>}
 pelvis {parent=root, loc=<0 -0.5301 98.0571>, rot=<0.0127 0.707 -0.0127 -0.707>}
 spine_01 {parent=pelvis, loc=<6.1208 -0.9328 0>, rot=<0 0 0.034 -0.9994>}
 spine_02 {parent=spine_01, loc=<6.5997 -0.2095 0>, rot=<0 0 0.0238 -0.9997>}
 spine_03 {parent=spine_02, loc=<6.4914 -0.1237 0>, rot=<0 0 0.0149 -0.9999>}
 spine_04 {parent=spine_03, loc=<6.9598 0.0794 0>, rot=<0 0 -0.0132 -0.9999>}
 spine_05 {parent=spine_04, loc=<6.9987 0 0>, rot=<0 0 -0.0015 -1>}
 neck_01 {parent=spine_05, loc=<11.3138 0.1529 0>, rot=<0 0 -0.2003 -0.9797>}
 neck_02 {parent=neck_01, loc=<4.5627 0 0>, rot=<0 0 0.0504 -0.9987>}
 head {parent=neck_02, loc=<5.3914 0 0>, rot=<0 0 0.115 -0.9934>}
 clavicle_l {parent=spine_05, loc=<5.4831 -0.1053 -3.2946>, rot=<0.0465 -0.7289 -0.009 -0.6829>}
 upperarm_l {parent=clavicle_l, loc=<9.8658 -1.2136 -0.4105>, rot=<0.0164 -0.3245 -0.0269 -0.9454>}
 lowerarm_l {parent=upperarm_l, loc=<26.465 -0.8345 0.259>, rot=<0.0036 0.0161 -0.2298 -0.9731>}
 hand_l {parent=lowerarm_l, loc=<22.7355 -0.3203 -0.6105>, rot=<-0.6784 -0.0223 0.0735 -0.7307>}
 thumb_01_l {parent=hand_l, loc=<1.3118 -0.3191 -2.459>, rot=<0.6546 -0.4042 0.0664 -0.6354>}
 thumb_02_l {parent=thumb_01_l, loc=<5.9336 -0.0448 0.0401>, rot=<-0.0037 0.0213 0.298 -0.9543>}
 thumb_03_l {parent=thumb_02_l, loc=<2.2223 0.3682 -0.0907>, rot=<0.0118 -0.013 0.0066 -0.9998>}
 index_metacarpal_l {parent=hand_l, loc=<1.6535 0.0183 -2.96>, rot=<0.0158 -0.0124 -0.0173 -0.9996>}
 index_01_l {parent=index_metacarpal_l, loc=<8.3008 0.0357 -0.0524>, rot=<-0.0044 0.0485 0.0924 -0.9945>}
 index_02_l {parent=index_01_l, loc=<3.3587 -0.3004 0.0452>, rot=<-0.002 0.0027 0.3225 -0.9466>}
 index_03_l {parent=index_02_l, loc=<2.5578 -0.123 -0.0252>, rot=<-0.002 0.0067 0.2558 -0.9667>}
 middle_metacarpal_l {parent=hand_l, loc=<2.1856 -0.1725 -0.5984>, rot=<0.006 0.0168 -0.0329 -0.9993>}
 middle_01_l {parent=middle_metacarpal_l, loc=<7.3275 -0.2447 0.0129>, rot=<-0.0208 0.0042 0.0858 -0.9961>}
 middle_02_l {parent=middle_01_l, loc=<4.0123 0.1401 0.1175>, rot=<-0.0218 0.0258 0.3512 -0.9357>}
 middle_03_l {parent=middle_02_l, loc=<2.6077 0.0534 0.1373>, rot=<0.0082 0.0188 0.1708 -0.9851>}
 ring_metacarpal_l {parent=hand_l, loc=<2.2074 -0.3625 1.3638>, rot=<-0.0284 0.0722 -0.0458 -0.9959>}
 ring_01_l {parent=ring_metacarpal_l, loc=<7.0295 -0.6197 -0.1446>, rot=<-0.0489 -0.0252 0.1892 -0.9804>}
 ring_02_l {parent=ring_01_l, loc=<3.6174 0.4647 -0.1027>, rot=<0 0 0.2806 -0.9598>}
 ring_03_l {parent=ring_02_l, loc=<2.4152 0.114 0.0199>, rot=<0 0 0.1189 -0.9929>}
 pinky_metacarpal_l {parent=hand_l, loc=<1.7616 -0.9753 3.3814>, rot=<-0.0397 0.0814 -0.069 -0.9935>}
 pinky_01_l {parent=pinky_metacarpal_l, loc=<7.1709 -0.4831 -0.1094>, rot=<-0.0654 -0.0035 0.2242 -0.9723>}
 pinky_02_l {parent=pinky_01_l, loc=<2.8274 0.0893 -0.1043>, rot=<0.0037 0.0011 0.2854 -0.9584>}
 pinky_03_l {parent=pinky_02_l, loc=<2.2046 0.0307 -0.0078>, rot=<-0.0002 0.0098 0.1445 -0.9895>}
 deform_wrist_dn_base_l {parent=hand_l, loc=<0 -1.5326 0>, rot=<0 0 0 -1>}
 deform_wrist_dn_l {parent=deform_wrist_dn_base_l, loc=<0 0 0>, rot=<0 0 0 -1>}
 deform_wrist_up_base_l {parent=hand_l, loc=<0 2.3009 0>, rot=<0 0 0 -1>}
 deform_wrist_up_l {parent=deform_wrist_up_base_l, loc=<0 0 0>, rot=<0 0 0 -1>}
 hand_attach_l {parent=hand_l, loc=<8.5 -3 0>, rot=<0.7071 0.7071 0 0>}
 weapon_l {parent=hand_l, loc=<-0.1974 -3.4111 0.3811>, rot=<0 0 0 1>}
 lowerarm_twist_02_l {parent=lowerarm_l, loc=<10.6707 0 0>, rot=<0 0 0 1>}
 lowerarm_twist_01_l {parent=lowerarm_l, loc=<18.316 0 0>, rot=<0 0 0 1>}
 deform_elbow_in_base_l {parent=lowerarm_l, loc=<0.1834 3.6062 -0.2572>, rot=<0 0 0 -1>}
 deform_elbow_in_l {parent=deform_elbow_in_base_l, loc=<0 0 0>, rot=<0 0 0 -1>}
 deform_elbow_out_base_l {parent=lowerarm_l, loc=<0 -4 0>, rot=<0 0 0 1>}
 deform_elbow_out_l {parent=deform_elbow_out_base_l, loc=<0 0 0>, rot=<0 0 0 1>}
 upperarm_twist_01_l {parent=upperarm_l, loc=<7.2135 -0.1134 0>, rot=<0 0 0 1>}
 deform_upperarm_bk_base_l {parent=upperarm_twist_01_l, loc=<-6.4053 -4.3746 0.314>, rot=<0 0 0 -1>}
 deform_upperarm_bk_l {parent=deform_upperarm_bk_base_l, loc=<0 0 0>, rot=<0 0 0 -1>}
 deform_upperarm_fr_base_l {parent=upperarm_twist_01_l, loc=<-5.6312 5.092 0.3354>, rot=<0 0 0 -1>}
 deform_upperarm_fr_l {parent=deform_upperarm_fr_base_l, loc=<0 0 0>, rot=<0 0 0 -1>}
 deform_upperarm_up_base_l {parent=upperarm_twist_01_l, loc=<-8.891 0.4825 3.9517>, rot=<0 0 0 -1>}
 deform_upperarm_up_l {parent=deform_upperarm_up_base_l, loc=<0 0 0>, rot=<0 0 0 -1>}
 Shoulder_twist_l {parent=upperarm_twist_01_l, loc=<-3.2702 0 1.5746>, rot=<0 0 0 1>}
 upperarm_twist_02_l {parent=upperarm_l, loc=<19.7998 -0.1134 0>, rot=<0 0 0 1>}
 Shoulders_main_02_l {parent=upperarm_l, loc=<-0.0002 0 0.0003>, rot=<0 0 0 1>}
 Shoulders_02_l {parent=Shoulders_main_02_l, loc=<0 0 0>, rot=<0 0 0 1>}
 Shoulders_01_l {parent=clavicle_l, loc=<3.4943 0 5.3228>, rot=<0 -0.1084 0 -0.9941>}
 Shoulders_base_aim_l {parent=clavicle_l, loc=<1.166 -1.6534 -0.995>, rot=<0.0107 -0.1277 -0.0296 -0.9913>}
 Shoulders_aim_l {parent=Shoulders_base_aim_l, loc=<0 0 0>, rot=<0 -0.2011 0 -0.9796>}
 clavicle_r {parent=spine_05, loc=<5.4831 -0.1053 3.2946>, rot=<0.7289 0.0465 -0.6829 0.009>}
 upperarm_r {parent=clavicle_r, loc=<-9.8658 1.2136 0.4105>, rot=<0.0164 -0.3245 -0.0269 -0.9454>}
 lowerarm_r {parent=upperarm_r, loc=<-26.465 0.8345 -0.259>, rot=<0.0036 0.0161 -0.2298 -0.9731>}
 hand_r {parent=lowerarm_r, loc=<-22.7355 0.3203 0.6106>, rot=<-0.6784 -0.0223 0.0735 -0.7307>}
 thumb_01_r {parent=hand_r, loc=<-1.3118 0.3191 2.459>, rot=<0.6546 -0.4042 0.0664 -0.6354>}
 thumb_02_r {parent=thumb_01_r, loc=<-5.9336 0.0448 -0.0401>, rot=<-0.0037 0.0213 0.298 -0.9543>}
 thumb_03_r {parent=thumb_02_r, loc=<-2.2223 -0.3682 0.0907>, rot=<0.0118 -0.013 0.0066 -0.9998>}
 index_metacarpal_r {parent=hand_r, loc=<-1.6535 -0.0183 2.96>, rot=<0.0158 -0.0124 -0.0173 -0.9996>}
 index_01_r {parent=index_metacarpal_r, loc=<-8.3008 -0.0357 0.0524>, rot=<-0.0044 0.0485 0.0924 -0.9945>}
 index_02_r {parent=index_01_r, loc=<-3.3587 0.3004 -0.0452>, rot=<-0.002 0.0027 0.3225 -0.9466>}
 index_03_r {parent=index_02_r, loc=<-2.5578 0.123 0.0252>, rot=<-0.002 0.0067 0.2558 -0.9667>}
 middle_metacarpal_r {parent=hand_r, loc=<-2.1856 0.1725 0.5984>, rot=<0.006 0.0168 -0.0329 -0.9993>}
 middle_01_r {parent=middle_metacarpal_r, loc=<-7.3275 0.2447 -0.0129>, rot=<-0.0208 0.0042 0.0858 -0.9961>}
 middle_02_r {parent=middle_01_r, loc=<-4.0123 -0.1401 -0.1175>, rot=<-0.0218 0.0258 0.3512 -0.9357>}
 middle_03_r {parent=middle_02_r, loc=<-2.6077 -0.0534 -0.1373>, rot=<0.0082 0.0188 0.1708 -0.9851>}
 ring_metacarpal_r {parent=hand_r, loc=<-2.2074 0.3625 -1.3638>, rot=<-0.0284 0.0722 -0.0458 -0.9959>}
 ring_01_r {parent=ring_metacarpal_r, loc=<-7.0295 0.6197 0.1446>, rot=<-0.0489 -0.0252 0.1892 -0.9804>}
 ring_02_r {parent=ring_01_r, loc=<-3.6174 -0.4647 0.1027>, rot=<0 0 0.2806 -0.9598>}
 ring_03_r {parent=ring_02_r, loc=<-2.4152 -0.114 -0.0199>, rot=<0 0 0.1189 -0.9929>}
 pinky_metacarpal_r {parent=hand_r, loc=<-1.7616 0.9753 -3.3814>, rot=<-0.0397 0.0814 -0.069 -0.9935>}
 pinky_01_r {parent=pinky_metacarpal_r, loc=<-7.1709 0.4831 0.1094>, rot=<-0.0654 -0.0035 0.2242 -0.9723>}
 pinky_02_r {parent=pinky_01_r, loc=<-2.8274 -0.0893 0.1043>, rot=<0.0037 0.0011 0.2854 -0.9584>}
 pinky_03_r {parent=pinky_02_r, loc=<-2.2046 -0.0307 0.0078>, rot=<-0.0002 0.0098 0.1445 -0.9895>}
 deform_wrist_dn_base_r {parent=hand_r, loc=<-0.0001 1.5328 0>, rot=<0 0 0 -1>}
 deform_wrist_dn_r {parent=deform_wrist_dn_base_r, loc=<0 0 0>, rot=<0 0 0 -1>}
 deform_wrist_up_base_r {parent=hand_r, loc=<-0.0002 -2.3007 -0.0001>, rot=<0 0 0 -1>}
 deform_wrist_up_r {parent=deform_wrist_up_base_r, loc=<0 0 0>, rot=<0 0 0 -1>}
 weapon_r {parent=hand_r, loc=<0.1974 3.4111 -0.3811>, rot=<0 0 0 1>}
 hand_attach_r {parent=hand_r, loc=<-8.5001 2.9999 0>, rot=<0 0 -0.7071 -0.7071>}
 lowerarm_twist_02_r {parent=lowerarm_r, loc=<-10.6707 0 0>, rot=<0 0 0 -1>}
 lowerarm_twist_01_r {parent=lowerarm_r, loc=<-18.316 0 0>, rot=<0 0 0 -1>}
 deform_elbow_in_base_r {parent=lowerarm_r, loc=<-0.1834 -3.6062 0.2572>, rot=<0 0 0 -1>}
 deform_elbow_in_r {parent=deform_elbow_in_base_r, loc=<0 0 0>, rot=<0 0 0 -1>}
 deform_elbow_out_base_r {parent=lowerarm_r, loc=<0 4 -0.0001>, rot=<0 0 0 -1>}
 deform_elbow_out_r {parent=deform_elbow_out_base_r, loc=<0 0 0>, rot=<0 0 0 1>}
 upperarm_twist_01_r {parent=upperarm_r, loc=<-7.2135 0.1134 0>, rot=<0 0 0 -1>}
 deform_upperarm_bk_base_r {parent=upperarm_twist_01_r, loc=<6.4053 4.3746 -0.314>, rot=<0 0 0 -1>}
 deform_upperarm_bk_r {parent=deform_upperarm_bk_base_r, loc=<0 0 0>, rot=<0 0 0 -1>}
 deform_upperarm_fr_base_r {parent=upperarm_twist_01_r, loc=<5.6312 -5.092 -0.3354>, rot=<0 0 0 -1>}
 deform_upperarm_fr_r {parent=deform_upperarm_fr_base_r, loc=<0 0 0>, rot=<0 0 0 -1>}
 deform_upperarm_up_base_r {parent=upperarm_twist_01_r, loc=<8.891 -0.4825 -3.9517>, rot=<0 0 0 -1>}
 deform_upperarm_up_r {parent=deform_upperarm_up_base_r, loc=<0 0 0>, rot=<0 0 0 -1>}
 Shoulder_twist_r {parent=upperarm_twist_01_r, loc=<3.27 0 -1.5743>, rot=<0 0 0 1>}
 upperarm_twist_02_r {parent=upperarm_r, loc=<-19.7998 0.1134 0>, rot=<0 0 0 -1>}
 Shoulders_main_02_r {parent=upperarm_r, loc=<0.0002 0 -0.0003>, rot=<0 0 0 1>}
 Shoulders_02_r {parent=Shoulders_main_02_r, loc=<0 0 0>, rot=<0 0 0 -1>}
 Shoulders_01_r {parent=clavicle_r, loc=<-3.4943 0 -5.3224>, rot=<0 -0.1084 0 -0.9941>}
 Shoulders_base_aim_r {parent=clavicle_r, loc=<-1.166 1.6534 0.995>, rot=<0.0107 -0.1277 -0.0296 -0.9913>}
 Shoulders_aim_r {parent=Shoulders_base_aim_r, loc=<0 0 0>, rot=<0 -0.2011 0 -0.9796>}
 attach_backpack {parent=spine_05, loc=<-0.123 -13.473 0>, rot=<-0.0093 -0.707 0.0093 -0.707>}
 Backpack_BR {parent=attach_backpack, type=Null, loc=<0 3.666 0>, rot=<0 0 0 1>}
 attach_cape {parent=spine_05, loc=<9.017 -11.714 0>, rot=<0.0616 -0.7044 -0.0616 -0.7044>}
 cape {parent=attach_cape, type=Null, loc=<0 1.733 -5.2187>, rot=<0.1224 0 0 0.9925>}
 deform_pec_aim_base_l {parent=spine_04, loc=<3.2456 9.3309 -1.2237>, rot=<0.0603 -0.4033 0.3318 -0.8507>}
 deform_pec_aim_l {parent=deform_pec_aim_base_l, loc=<0 0 0>, rot=<0 0 0 1>}
 deform_pec_base_l {parent=deform_pec_aim_l, loc=<9.3665 0 0>, rot=<0 0 0 1>}
 deform_pec_l {parent=deform_pec_base_l, loc=<0 0 0>, rot=<0 0 0 -1>}
 deform_pec_aim_base_r {parent=spine_04, loc=<3.2452 9.3309 1.2237>, rot=<-0.4033 -0.0603 0.8507 0.3318>}
 deform_pec_aim_r {parent=deform_pec_aim_base_r, loc=<0 0 0>, rot=<0 0 0 1>}
 deform_pec_base_r {parent=deform_pec_aim_r, loc=<-9.3667 -0.0003 -0.0004>, rot=<0 0 0 1>}
 deform_pec_r {parent=deform_pec_base_r, loc=<0 0 0>, rot=<0 0 0 -1>}
 pouche_l {parent=pelvis, loc=<10.929 -4.6562 -8.8583>, rot=<0.1813 -0.8992 -0.384 -0.1053>}
 thigh_l {parent=pelvis, loc=<-1.0855 -0.2109 -8.7053>, rot=<0.0009 0.0355 0.0155 -0.9993>}
 calf_l {parent=thigh_l, loc=<-42.5284 -0.0019 -0.7297>, rot=<0.0006 -0.0161 -0.0391 -0.9991>}
 foot_l {parent=calf_l, loc=<-46.7401 -2.5014 -0.592>, rot=<0.0427 -0.0238 0.0353 -0.9982>}
 ball_l {parent=foot_l, loc=<-3.5658 10.8066 -0.1544>, rot=<0.0001 -0.0001 -0.7063 -0.7079>}
 boots_cor_01_l1 {parent=foot_l, loc=<1.3894 8.4579 0.2723>, rot=<-0.0281 0.0276 0.2421 -0.9695>}
 calf_twist_02_l {parent=calf_l, loc=<-19.1546 -1.7943 -0.2638>, rot=<0.0237 0.0095 -0.0035 -0.9997>}
 calf_twist_01_l {parent=calf_l, loc=<-35.8483 -1.9197 -0.5794>, rot=<0.0237 0.0095 -0.0035 -0.9997>}
 boots_cor_01_l {parent=calf_twist_01_l, loc=<-9.8507 -3.543 0>, rot=<0 0 -0.038 -0.9993>}
 deform_knee_in_base_l {parent=calf_l, loc=<0.8031 -5.0231 0.2456>, rot=<0 0 0 -1>}
 deform_knee_in_l {parent=deform_knee_in_base_l, loc=<0 0 0>, rot=<0 0 0 1>}
 deform_knee_out_base_l {parent=calf_l, loc=<0 4 0>, rot=<0 0 0 -1>}
 deform_knee_out_l {parent=deform_knee_out_base_l, loc=<0 0 0>, rot=<0 0 0 1>}
 Kneepad_l {parent=calf_l, loc=<0 0 0>, rot=<0 0 0 1>}
 thigh_twist_01_l {parent=thigh_l, loc=<-27.952 0 0>, rot=<0.0013 0 0 -1>}
 deform_groin_base_l {parent=thigh_l, loc=<0.4801 5.6776 -0.2486>, rot=<0 0 0 -1>}
 deform_groin_l {parent=deform_groin_base_l, loc=<0 0 0>, rot=<0 0 0 -1>}
 thigh_r {parent=pelvis, loc=<-1.0855 -0.2109 8.7053>, rot=<0.0355 -0.0009 0.9993 0.0155>}
 calf_r {parent=thigh_r, loc=<42.5284 0.0019 0.7297>, rot=<0.0006 -0.0161 -0.0391 -0.9991>}
 foot_r {parent=calf_r, loc=<46.7401 2.5014 0.592>, rot=<0.0427 -0.0238 0.0353 -0.9982>}
 ball_r {parent=foot_r, loc=<3.5658 -10.8066 0.1544>, rot=<0.0001 -0.0001 -0.7063 -0.7079>}
 boots_cor_01_r1 {parent=foot_r, loc=<-1.3894 -8.4579 -0.2723>, rot=<-0.0281 0.0276 0.2421 -0.9695>}
 calf_twist_02_r {parent=calf_r, loc=<19.1546 1.7943 0.2638>, rot=<0.0237 0.0095 -0.0035 -0.9997>}
 calf_twist_01_r {parent=calf_r, loc=<35.8483 1.9197 0.5794>, rot=<0.0237 0.0095 -0.0035 -0.9997>}
 boots_cor_01_r {parent=calf_twist_01_r, loc=<9.8507 3.543 0>, rot=<0 0 -0.038 -0.9993>}
 deform_knee_in_base_r {parent=calf_r, loc=<-0.8031 5.0231 -0.2457>, rot=<0 0 0 -1>}
 deform_knee_in_r {parent=deform_knee_in_base_r, loc=<0 0 0>, rot=<0 0 0 -1>}
 deform_knee_out_base_r {parent=calf_r, loc=<0 -4 0>, rot=<0 0 0 -1>}
 deform_knee_out_r {parent=deform_knee_out_base_r, loc=<0 0 0>, rot=<0 0 0 1>}
 Kneepad_r {parent=calf_r, loc=<0 0 0>, rot=<0 0 0 1>}
 thigh_twist_01_r {parent=thigh_r, loc=<27.952 0 0>, rot=<0.0013 0 0 -1>}
 deform_groin_base_r {parent=thigh_r, loc=<-0.4801 -5.6776 0.2486>, rot=<0 0 0 -1>}
 deform_groin_r {parent=deform_groin_base_r, loc=<0 0 0>, rot=<0 0 0 -1>}
 deform_glute_base_l {parent=pelvis, loc=<-7.3466 -9.0213 -9.1245>, rot=<0.0172 0.031 -0.4477 -0.8935>}
 deform_glute_l {parent=deform_glute_base_l, loc=<0 0 0>, rot=<0 0 0 1>}
 deform_glute_base_r {parent=pelvis, loc=<-7.3464 -9.0213 9.1245>, rot=<0.031 -0.0172 0.8935 -0.4477>}
 deform_glute_r {parent=deform_glute_base_r, loc=<0 0 0>, rot=<0 0 0 1>}
 pouche_r {parent=pelvis, loc=<10.9291 -4.6562 8.8583>, rot=<0.1813 -0.8992 0.384 0.1053>}
 attach {parent=root, loc=<0 0 0>, rot=<0 0 0 1>}
 ik_hand_root {parent=root, loc=<0 0 0>, rot=<0 0 0 1>}
 ik_hand_gun {parent=ik_hand_root, loc=<-48.1743 3.3545 104.4381>, rot=<0.6276 0.1309 0.4183 -0.6434>}
 ik_hand_r {parent=ik_hand_gun, loc=<0 0 0>, rot=<0 0 0 1>}
 ik_hand_l {parent=ik_hand_gun, loc=<59.3271 67.6997 34.3524>, rot=<-0.6158 -0.7027 -0.3565 0>}
 ik_foot_root {parent=root, loc=<0 0 0>, rot=<0 0 0 1>}
 ik_foot_l {parent=ik_foot_root, loc=<14.8262 -7.3545 8.296>, rot=<0.0357 0.7045 0.0284 -0.7082>}
 ik_foot_r {parent=ik_foot_root, loc=<-14.8262 -7.3545 8.296>, rot=<0.7082 0.0284 -0.7045 0.0357>}
 root_fp {parent=root, loc=<0 0 0>, rot=<0 0 0 1>}
]
variables [
 ISSkydive {type=double, default=0}
 isSkyDiving {type=bool, default=False}
]
