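# Animation Blueprint for the F_MED_UniqueBeetle player body cosmetic, adding secondary motion on top of the pose copied from the attached parent mesh.
# Spring-driven Anim Dynamics nodes with angular limits and overridden gravity simulate the left and right pouch bones in series.
# The result feeds a shoulder Control Rig, which receives skydive state as both a bool and a numeric IsSkyDive toggled by skydiving begin/end events.
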AnimBlueprint /BRCosmetics/Characters/Player/Female/Medium/Bodies/F_MED_UniqueBeetle/Meshes/F_MED_UniqueBeetle_AnimBP
target_skeleton=&/F_MED_UniqueBeetle_Skeleton
graphs [
 AnimGraph {nodes [
   n1 = Anim Dynamics - Bone: pouche_l(ComponentPose=Local To Component(LocalPose=Copy Pose From Mesh() {bUseAttachedParent=true}), Alpha=1, ExternalForce="0,0,0") {
    angularDampingOverride=0.999
    angularSpringConstant=6000
    bAngularSpring=true
    bGravityOverrideInSimSpace=true
    bOverrideAngularDamping=true
    bUseGravityOverride=true
    boundBone {boneName=pouche_l}
    componentAppliedLinearAccClamp=<3000 3000 2500>
    componentLinearAccScale=<0.5 0.5 0.5>
    componentLinearVelScale=<0.5 0.5 0.5>
    gravityOverride {z=-980}
    physicsBodyDefinitions [{
      boundBone {boneName=pouche_l}
      boxExtents=<15 14 14>
      constraintSetup {angularLimitsMax=<0 3 0>, angularLimitsMin=<0 -3 -45>, angularTarget=<0 0 0>}
      localJointOffset=<1 0 0>
     }]
    simSpaceSettings {maxAngularAcceleration=35, maxAngularVelocity=7, simSpaceAngularAlpha=1}
   }
   n2 = Anim Dynamics - Bone: pouche_r(ComponentPose=@n1, Alpha=1, ExternalForce="0,0,0") {
    angularDampingOverride=0.999
    angularSpringConstant=6000
    bAngularSpring=true
    bGravityOverrideInSimSpace=true
    bOverrideAngularDamping=true
    bUseGravityOverride=true
    boundBone {boneName=pouche_r}
    componentAppliedLinearAccClamp=<3000 3000 2500>
    componentLinearAccScale=<0.5 0.5 0.5>
    componentLinearVelScale=<0.5 0.5 0.5>
    gravityOverride {z=-980}
    physicsBodyDefinitions [{
      boundBone {boneName=pouche_r}
      boxExtents=<14 13 13>
      constraintSetup {angularLimitsMax=<0 3 0>, angularLimitsMin=<0 -3 -45>, angularTarget=<0 0 0>}
      localJointOffset=<1 0 0>
     }]
    simSpaceSettings {maxAngularAcceleration=35, maxAngularVelocity=7, simSpaceAngularAlpha=1}
   }
   n3 = Control Rig!(
    Source=Component To Local(ComponentPose=@n2)
    Alpha
    ISSkydive:double=Get IsSkyDive() -> (IsSkyDive:double)
    isSkyDiving:bool=Get bIsSkyDiving()
   ) -> (Pose) {controlRigAssetReference {blueprintRigClass=&/F_MED_UniqueBeetle_CtrlRig_Shoulder.F_MED_UniqueBeetle_CtrlRig_Shoulder_C}}
   Output Pose(Result=@n3) {layerGroup=None}
  ]}
 EventGraph {nodes [
   Event OnBeginSkydiving()
   > Set IsSkyDive(IsSkyDive:double=1) -> (Output_Get:double)
   Event OnEndSkydiving()
   > Set IsSkyDive(IsSkyDive=0)
   Event BlueprintUpdateAnimation()
  ]}
]
variables [
 IsSkyDive {type=number}
 NewVar {type=boolean}
]
