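# Physics asset for the face accessory mesh of the female medium UniqueBeetle character body, defining capsule bodies for a ponytail and upper body.
# Five ponytail bones form a constrained chain rooted at the head, with swing limits widening toward the tip and twist kept tight.
# Larger capsules cover the head, neck_02 and spine_05, joined by constraints with free swing and twist.
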PhysicsAsset /BRCosmetics/Characters/Player/Female/Medium/Bodies/F_MED_UniqueBeetle/Meshes/Parts/F_MED_UniqueBeetle_FaceAcc_Physics
bodies {
 PonyTail_phy_01 [{shapeType=Capsule, length=1.9618, radius=2.3329, center=<1.6672 -1.2335 0>}]
 PonyTail_phy_02 [{shapeType=Capsule, length=0.75, radius=2.8125, shapeName=PonyTail_phy_02_capsule, center=<3.1112 -0.4976 0>, rotation=<rot 90 0.0008 0>}]
 PonyTail_phy_03 [{shapeType=Capsule, length=2, radius=2.1875, shapeName=PonyTail_phy_03_capsule, center=<2.9946 -0.3338 0>, rotation=<rot 90 0.0008 0>}]
 PonyTail_phy_04 [{shapeType=Capsule, length=3.375, radius=1.375, shapeName=PonyTail_phy_04_capsule, center=<3.0198 -0.1117 0>, rotation=<rot 90 0.001 0>}]
 PonyTail_phy_05 [{shapeType=Capsule, length=3.9375, radius=1.1875, shapeName=PonyTail_phy_05_capsule, center=<3.0455 -0.1008 0>, rotation=<rot -90 0.0004 0>}]
 head [
  {shapeType=Capsule, length=11.2111, radius=5.9009, center=<3.6523 1.5642 -0.0633>, rotation=<rot -0.1023 -82.5783 -89.6182>}
  {shapeType=Capsule, length=2.5, radius=7.5, shapeName=head_capsule, center=<7.3592 -0.8179 0>, rotation=<rot 90 29.9984 0>}
 ]
 neck_02 [{shapeType=Capsule, length=9.505, radius=4.505, shapeName=neck_02_capsule, rotation=<rot 0 -90.0005 -90>}]
 spine_05 [{shapeType=Capsule, length=15.005, radius=11.255, shapeName=spine_05_capsule, center=<1.473 0.9632 0>, rotation=<rot 0 -90 0>}]
}
constraints [
 {bone1Name=PonyTail_phy_01, bone2Name=head, swing1LimitDegrees=10, swing2LimitDegrees=5, twistLimitDegrees=2}
 {bone1Name=PonyTail_phy_02, bone2Name=PonyTail_phy_01, swing1LimitDegrees=30, swing2LimitDegrees=15, twistLimitDegrees=5}
 {bone1Name=PonyTail_phy_03, bone2Name=PonyTail_phy_02, swing1LimitDegrees=35, swing2LimitDegrees=25, twistLimitDegrees=5}
 {bone1Name=PonyTail_phy_04, bone2Name=PonyTail_phy_03, swing1LimitDegrees=45, swing2LimitDegrees=35, twistLimitDegrees=5}
 {bone1Name=PonyTail_phy_05, bone2Name=PonyTail_phy_04, swing1LimitDegrees=45, swing2LimitDegrees=45, twistLimitDegrees=5}
 {
  bone1Name=head
  bone2Name=neck_02
  swing1LimitDegrees=45
  swing1Motion=Free
  swing2LimitDegrees=45
  swing2Motion=Free
  twistLimitDegrees=45
  twistMotion=Free
 }
 {
  bone1Name=neck_02
  bone2Name=spine_05
  swing1LimitDegrees=45
  swing1Motion=Free
  swing2LimitDegrees=45
  swing2Motion=Free
  twistLimitDegrees=45
  twistMotion=Free
 }
]
