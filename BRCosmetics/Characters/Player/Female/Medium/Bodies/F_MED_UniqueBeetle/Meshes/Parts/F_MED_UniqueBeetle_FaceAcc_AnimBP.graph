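# Post-process style secondary-motion Anim Blueprint for the UniqueBeetle female face accessory part, adding physics to its hair bones.
# AnimGraph copies the attached parent mesh pose, then runs RigidBody simulation on the part's physics asset in spine_05 space.
# Six half-alpha, root-relative AnimDynamics nodes then chain over left and right Hair_phy bones with angular springs and limits.
# Each update, a nativized function fills gravity-override and RigidBody-settings variables from a parameter asset and game time.
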AnimBlueprint /BRCosmetics/Characters/Player/Female/Medium/Bodies/F_MED_UniqueBeetle/Meshes/Parts/F_MED_UniqueBeetle_FaceAcc_AnimBP
target_skeleton=&/F_MED_UniqueBeetle_FaceAcc_Skeleton
graphs [
 AnimGraph {nodes [
   n1 = RigidBody!(
    ComponentPose=Local To Component(LocalPose=Copy Pose From Mesh() {bUseAttachedParent=true})
    Alpha=0.8
    OverrideWorldGravity:Vector="0,0,0"
    ExternalForce
    SimSpaceSettings:Sim Space Settings Structure=(WorldAlpha=1,VelocityScaleZ=0.5,DampingAlpha=1,MaxLinearVelocity=800,MaxAngularVelocity=8,MaxLinearAcceleration=2400,MaxAngularAcceleration=32,ExternalLinearDragV=(X=0.25,Y=0.25,Z=0.25),ExternalLinearVelocity=(X=0,Y=0,Z=0),ExternalAngularVelocity=(X=0,Y=0,Z=0))
   ) -> (Pose) {
    bClampLinearTranslationLimitToRefPose=true
    bOverrideWorldGravity=true
    overridePhysicsAsset=&/F_MED_UniqueBeetle_FaceAcc_Physics
    simulationSpace=BaseBoneSpace
    alphaBoolBlend {blendOption=Linear}
    baseBoneRef {boneName=spine_05}
    componentAppliedLinearAccClamp=<3000 3000 3000>
    componentLinearAccScale=<0.2 0.2 0.2>
    componentLinearVelScale=<0.2 0.2 0.3>
    overrideWorldGravity {z=-850}
    simSpaceSettings {
     maxAngularAcceleration=32
     maxAngularVelocity=8
     maxLinearAcceleration=2400
     maxLinearVelocity=800
     velocityScaleZ=0.5
     worldAlpha=1
     externalLinearDragV=<0.25 0.25 0.25>
    }
   }
   n2 = Anim Dynamics - Bone: Hair_phy_01_l!(ComponentPose=@n1, ExternalForce="0,0,0") -> (Pose) {
    alpha=0.5
    angularDampingOverride=0.99
    angularSpringConstant=100
    bAngularSpring=true
    bGravityOverrideInSimSpace=true
    bOverrideAngularDamping=true
    gravityScale=0.35
    simulationSpace=RootRelative
    alphaBoolBlend {blendOption=Linear}
    boundBone {boneName=Hair_phy_01_l}
    componentAppliedLinearAccClamp=<3000 3000 3000>
    componentLinearAccScale {x=0.5, z=0.6}
    componentLinearVelScale {x=0.5, z=0.6}
    physicsBodyDefinitions [{
      sphereCollisionRadius=0
      boundBone {boneName=Hair_phy_01_l}
      boxExtents=<8 30 30>
      constraintSetup {angularLimitsMax=<0 0 5>, angularLimitsMin=<0 -20 -30>, angularTarget=<1 0 0>}
      localJointOffset=<6 0 0>
     }]
    simSpaceSettings {maxAngularAcceleration=35, maxAngularVelocity=7, simSpaceAngularAlpha=0.3}
   }
   n3 = Anim Dynamics - Bone: Hair_phy_02_l!(ComponentPose=@n2, ExternalForce="0,0,0") -> (Pose) {
    alpha=0.5
    angularDampingOverride=0.99
    angularSpringConstant=100
    bAngularSpring=true
    bGravityOverrideInSimSpace=true
    bOverrideAngularDamping=true
    gravityScale=0.35
    simulationSpace=RootRelative
    alphaBoolBlend {blendOption=Linear}
    boundBone {boneName=Hair_phy_02_l}
    componentAppliedLinearAccClamp=<3000 3000 3000>
    componentLinearAccScale {x=0.5, z=0.6}
    componentLinearVelScale {x=0.5, z=0.6}
    physicsBodyDefinitions [{
      sphereCollisionRadius=0
      boundBone {boneName=Hair_phy_02_l}
      boxExtents=<8 30 30>
      constraintSetup {angularLimitsMax=<0 5 20>, angularLimitsMin=<0 -15 -60>, angularTarget=<1 0 0>}
      localJointOffset=<6 0 0>
     }]
    simSpaceSettings {maxAngularAcceleration=35, maxAngularVelocity=7, simSpaceAngularAlpha=0.5}
   }
   n4 = Anim Dynamics - Bone: Hair_phy_03_l!(ComponentPose=@n3, ExternalForce="0,0,0") -> (Pose) {
    alpha=0.5
    angularDampingOverride=0.995
    angularSpringConstant=100
    bAngularSpring=true
    bGravityOverrideInSimSpace=true
    bOverrideAngularDamping=true
    gravityScale=0.35
    simulationSpace=RootRelative
    alphaBoolBlend {blendOption=Linear}
    boundBone {boneName=Hair_phy_03_l}
    componentAppliedLinearAccClamp=<3000 3000 3000>
    componentLinearAccScale {x=0.5, z=0.6}
    componentLinearVelScale {x=0.5, z=0.6}
    physicsBodyDefinitions [{
      sphereCollisionRadius=0
      boundBone {boneName=Hair_phy_03_l}
      boxExtents=<8 30 30>
      constraintSetup {angularLimitsMax=<0 15 10>, angularLimitsMin=<0 -15 -60>, angularTarget=<1 0 0>}
      localJointOffset=<6 0 0>
     }]
    simSpaceSettings {maxAngularAcceleration=35, maxAngularVelocity=7, simSpaceAngularAlpha=0.6}
   }
   n5 = Anim Dynamics - Bone: Hair_phy_01_r!(ComponentPose=@n4, ExternalForce="0,0,0") -> (Pose) {
    alpha=0.5
    angularDampingOverride=0.99
    angularSpringConstant=100
    bAngularSpring=true
    bGravityOverrideInSimSpace=true
    bOverrideAngularDamping=true
    gravityScale=0.35
    simulationSpace=RootRelative
    alphaBoolBlend {blendOption=Linear}
    boundBone {boneName=Hair_phy_01_r}
    componentAppliedLinearAccClamp=<3000 3000 3000>
    componentLinearAccScale {x=0.5, z=0.6}
    componentLinearVelScale {x=0.5, z=0.6}
    physicsBodyDefinitions [{
      sphereCollisionRadius=0
      boundBone {boneName=Hair_phy_01_r}
      boxExtents=<8 30 30>
      constraintSetup {angularLimitsMax=<0 20 5>, angularLimitsMin=<0 0 -30>, angularTarget=<1 0 0>}
      localJointOffset=<6 0 0>
     }]
    simSpaceSettings {maxAngularAcceleration=35, maxAngularVelocity=7, simSpaceAngularAlpha=0.3}
   }
   n6 = Anim Dynamics - Bone: Hair_phy_02_r!(ComponentPose=@n5, ExternalForce="0,0,0") -> (Pose) {
    alpha=0.5
    angularDampingOverride=0.99
    angularSpringConstant=100
    bAngularSpring=true
    bGravityOverrideInSimSpace=true
    bOverrideAngularDamping=true
    gravityScale=0.35
    simulationSpace=RootRelative
    alphaBoolBlend {blendOption=Linear}
    boundBone {boneName=Hair_phy_02_r}
    componentAppliedLinearAccClamp=<3000 3000 3000>
    componentLinearAccScale {x=0.5, z=0.6}
    componentLinearVelScale {x=0.5, z=0.6}
    physicsBodyDefinitions [{
      sphereCollisionRadius=0
      boundBone {boneName=Hair_phy_02_r}
      boxExtents=<8 30 30>
      constraintSetup {angularLimitsMax=<0 15 20>, angularLimitsMin=<0 -5 -60>, angularTarget=<1 0 0>}
      localJointOffset=<6 0 0>
     }]
    simSpaceSettings {maxAngularAcceleration=35, maxAngularVelocity=7, simSpaceAngularAlpha=0.5}
   }
   n7 = Anim Dynamics - Bone: Hair_phy_03_r!(ComponentPose=@n6, ExternalForce="0,0,0") -> (Pose) {
    alpha=0.5
    angularDampingOverride=0.995
    angularSpringConstant=100
    bAngularSpring=true
    bGravityOverrideInSimSpace=true
    bOverrideAngularDamping=true
    gravityScale=0.35
    simulationSpace=RootRelative
    alphaBoolBlend {blendOption=Linear}
    boundBone {boneName=Hair_phy_03_r}
    componentAppliedLinearAccClamp=<3000 3000 3000>
    componentLinearAccScale {x=0.5, z=0.6}
    componentLinearVelScale {x=0.5, z=0.6}
    physicsBodyDefinitions [{
      sphereCollisionRadius=0
      boundBone {boneName=Hair_phy_03_r}
      boxExtents=<8 30 30>
      constraintSetup {angularLimitsMax=<0 15 10>, angularLimitsMin=<0 -15 -60>, angularTarget=<1 0 0>}
      localJointOffset=<6 0 0>
     }]
    simSpaceSettings {maxAngularAcceleration=35, maxAngularVelocity=7, simSpaceAngularAlpha=0.6}
   }
   Output Pose(Result=Component To Local(ComponentPose=@n7)) {layerGroup=None}
  ]}
 EventGraph {nodes [
   Event BlueprintUpdateAnimation()
   > n1 = "Gravity Override State Function (Parameter Asset) (Nativized)"(
    CharacterPartAnimInstance=Self-Reference()
    GameTimeInSeconds=GetGameTimeInSeconds()
    Parameters=&/F_MED_UniqueBeetle_FaceAcc_Parameters
   )
   > Set Gravity Override(Gravity Override=@n1)
   > Set RBAN Settings Out(RBAN Settings Out=@n1.OutRigidBodyAnimNodeParameters)
  ]}
]
variables [
 Gravity Override {type=Vector}
 RBAN Settings Out {type=FortRigidBodyAnimNodeParameters}
]
